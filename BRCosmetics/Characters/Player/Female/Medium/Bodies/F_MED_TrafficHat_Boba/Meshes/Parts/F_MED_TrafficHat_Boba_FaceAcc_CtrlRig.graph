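# Control Rig for the TrafficHat Boba female character's FaceAcc part, applying pose-driven corrective offsets to bow and hair bones.
# Spherical pose readers on head and clavicles weight ModifyTransforms nodes across three sequenced passes: general, crouching, and skydive floating.
# Float state variables multiply weights in the crouch and skydive passes.
# In the general pass, subtracting IsCrouchingAndTargetingOrSkydiveFloating suppresses clavicle-driven bow fixes, while head-driven fixes stay ungated.
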ControlRigBlueprint /BRCosmetics/Characters/Player/Female/Medium/Bodies/F_MED_TrafficHat_Boba/Meshes/Parts/F_MED_TrafficHat_Boba_FaceAcc_CtrlRig
graphs [
 RigVMFunctionLibrary {links [], nodes []}
 RigVMFunction_Sequence_1_ContainedGraph {nodes [
   n1 = RigVMFunctionEntryNode()
   RigVMFunction_Sequence = &RigVMFunction_Sequence(ExecuteContext=@n1.ExecuteContext)
   n2 = &RigVMFunction_Sequence(ExecuteContext=@RigVMFunction_Sequence.B)
   RigVMFunctionReturnNode(A=@RigVMFunction_Sequence.A, B=@n2.A, C=@n2.B)
  ]}
 RigVMModel {nodes [
   RigUnit_BeginExecution = &RigUnit_BeginExecution()
   RigVMFunction_Sequence = RigVMAggregateNode(ExecuteContext=@RigUnit_BeginExecution.ExecutePin)
   n1 = &RigUnit_SphericalPoseReader(
    ActiveRegionScaleFactors=(PositiveWidth=1,NegativeWidth=1,PositiveHeight=1.5,NegativeHeight=1)
    ActiveRegionSize=0.1
    Debug=(bDrawDebug=false,bDraw2D=False,bDrawLocalAxes=False,DebugScale=25,DebugSegments=20,DebugThickness=0.25)
    DriverAxis=(X=1,Y=0,Z=0)
    DriverItem=(Type=Bone,Name="head")
    FalloffRegionScaleFactors=(PositiveWidth=1,NegativeWidth=1,PositiveHeight=1,NegativeHeight=1)
    FalloffSize=0.27
    FlipHeightScaling=False
    FlipWidthScaling=False
    OptionalParentItem=(Type=Bone,Name="spine_05")
    RotationOffset=(X=70,Y=0,Z=-135)
    ExecutePin=@RigVMFunction_Sequence.A
   )
   > &RigUnit_ModifyTransforms(
    ItemToModify=((Item=(Type=Bone,Name="dyn_bow"),Transform=(Rotation=(X=0,Y=0,Z=0,W=1),Translation=(X=-1.5,Y=-1,Z=0),Scale3D=(X=1,Y=1,Z=1))))
    WeightMinimum=0
    Weight=@n1.OutputParam
   )
   > &RigUnit_ModifyTransforms(
    ItemToModify=((Item=(Type=Bone,Name="dyn_hair_strand_02_MSTR_back"),Transform=(Rotation=(X=0.0302,Y=-0.171,Z=0.171,W=0.9698),Translation=(X=0,Y=0,Z=0),Scale3D=(X=1,Y=1,Z=1))))
    WeightMinimum=0
    Weight=@n1.OutputParam
   )
   > &RigUnit_ModifyTransforms(
    ItemToModify=((Item=(Type=Bone,Name="dyn_hair_strand_03_MSTR_back"),Transform=(Rotation=(X=0,Y=0,Z=0,W=1),Translation=(X=-2,Y=0,Z=0),Scale3D=(X=1,Y=1,Z=1))))
    WeightMinimum=0
    Weight=@n1.OutputParam
   )
   > &RigUnit_ModifyTransforms(
    ItemToModify=((Item=(Type=Bone,Name="dyn_hair_strand_04_MSTR_back"),Transform=(Rotation=(X=0,Y=0,Z=0,W=1),Translation=(X=-2,Y=0,Z=0),Scale3D=(X=1,Y=1,Z=1))))
    WeightMinimum=0
    Weight=@n1.OutputParam
   )
   > &RigUnit_ModifyTransforms(
    ItemToModify=((Item=(Type=Bone,Name="dyn_hair_strand_05_MSTR_back"),Transform=(Rotation=(X=0,Y=0,Z=0,W=1),Translation=(X=-2,Y=0,Z=0),Scale3D=(X=1,Y=1,Z=1))))
    WeightMinimum=0
    Weight=@n1.OutputParam
   )
   > n2 = &RigUnit_SphericalPoseReader(
    ActiveRegionScaleFactors=(PositiveWidth=1,NegativeWidth=1,PositiveHeight=1,NegativeHeight=1)
    ActiveRegionSize=0.05
    Debug=(bDrawDebug=false,bDraw2D=False,bDrawLocalAxes=False,DebugScale=25,DebugSegments=20,DebugThickness=0.25)
    DriverAxis=(X=-1,Y=0,Z=0)
    DriverItem=(Type=Bone,Name="clavicle_r")
    FalloffRegionScaleFactors=(PositiveWidth=1,NegativeWidth=1,PositiveHeight=1,NegativeHeight=1)
    FalloffSize=0.18
    FlipHeightScaling=False
    FlipWidthScaling=False
    OptionalParentItem=(Type=Bone,Name="None")
    RotationOffset=(X=50,Y=-150,Z=100)
   )
   Subtract = &RigVMFunction_MathDoubleSub(A=@n2.OutputParam, B=RigVMVariableNode() {variable=IsCrouchingAndTargetingOrSkydiveFloating}.Value)
   > &RigUnit_ModifyTransforms(
    ItemToModify=((Item=(Type=Bone,Name="dyn_bow_01_r_MSTR"),Transform=(Rotation=(X=0,Y=0,Z=-0.342,W=0.9397),Translation=(X=0,Y=1,Z=0),Scale3D=(X=1,Y=1,Z=1))))
    WeightMinimum=0
    Weight=@Subtract.Result
   )
   > n3 = &RigUnit_SphericalPoseReader(
    ActiveRegionScaleFactors=(PositiveWidth=1,NegativeWidth=1,PositiveHeight=1,NegativeHeight=1)
    ActiveRegionSize=0.05
    Debug=(bDrawDebug=false,bDraw2D=False,bDrawLocalAxes=False,DebugScale=25,DebugSegments=20,DebugThickness=0.25)
    DriverAxis=(X=1,Y=0,Z=0)
    DriverItem=(Type=Bone,Name="clavicle_l")
    FalloffRegionScaleFactors=(PositiveWidth=1,NegativeWidth=1,PositiveHeight=1,NegativeHeight=1)
    FalloffSize=0.18
    FlipHeightScaling=False
    FlipWidthScaling=False
    OptionalParentItem=(Type=Bone,Name="None")
    RotationOffset=(X=60,Y=0,Z=-110)
   )
   > &RigUnit_ModifyTransforms(
    ItemToModify=((Item=(Type=Bone,Name="dyn_bow_01_l_MSTR"),Transform=(Rotation=(X=0,Y=0,Z=-0.4226,W=0.9063),Translation=(X=0,Y=-2,Z=0),Scale3D=(X=1,Y=1,Z=1))))
    WeightMinimum=0
    Weight=&RigVMFunction_MathDoubleSub(A=@n3.OutputParam, B=RigVMVariableNode() {variable=IsCrouchingAndTargetingOrSkydiveFloating}.Value).Result
   )
   > n4 = &RigUnit_SphericalPoseReader(
    ActiveRegionScaleFactors=(PositiveWidth=1,NegativeWidth=1,PositiveHeight=1,NegativeHeight=1)
    ActiveRegionSize=0.03
    Debug=(bDrawDebug=false,bDraw2D=False,bDrawLocalAxes=False,DebugScale=25,DebugSegments=20,DebugThickness=0.25)
    DriverAxis=(X=-1,Y=0,Z=0)
    DriverItem=(Type=Bone,Name="clavicle_r")
    FalloffRegionScaleFactors=(PositiveWidth=1,NegativeWidth=1,PositiveHeight=1,NegativeHeight=1)
    FalloffSize=0.1
    FlipHeightScaling=False
    FlipWidthScaling=False
    OptionalParentItem=(Type=Bone,Name="None")
    RotationOffset=(X=20,Y=-150,Z=140)
   )
   > &RigUnit_ModifyTransforms(
    ItemToModify=((Item=(Type=Bone,Name="dyn_bow_01_r_MSTR"),Transform=(Rotation=(X=0,Y=0,Z=-0.0872,W=0.9962),Translation=(X=0,Y=1,Z=0),Scale3D=(X=1,Y=1,Z=1))))
    WeightMinimum=0
    Weight=&RigVMFunction_MathDoubleSub(A=@n4.OutputParam, B=RigVMVariableNode() {variable=IsCrouchingAndTargetingOrSkydiveFloating}.Value).Result
   )
   > n5 = &RigUnit_SphericalPoseReader(
    ActiveRegionScaleFactors=(PositiveWidth=1,NegativeWidth=1,PositiveHeight=1,NegativeHeight=1)
    ActiveRegionSize=0.03
    Debug=(bDrawDebug=false,bDraw2D=False,bDrawLocalAxes=False,DebugScale=25,DebugSegments=20,DebugThickness=0.25)
    DriverAxis=(X=1,Y=0,Z=0)
    DriverItem=(Type=Bone,Name="clavicle_l")
    FalloffRegionScaleFactors=(PositiveWidth=1,NegativeWidth=1,PositiveHeight=1,NegativeHeight=1)
    FalloffSize=0.1
    FlipHeightScaling=False
    FlipWidthScaling=False
    OptionalParentItem=(Type=Bone,Name="None")
    RotationOffset=(X=25,Y=0,Z=-90)
   )
   > &RigUnit_ModifyTransforms(
    ItemToModify=((Item=(Type=Bone,Name="dyn_bow_01_l_MSTR"),Transform=(Rotation=(X=0,Y=0,Z=-0.0872,W=0.9962),Translation=(X=0,Y=-1,Z=0),Scale3D=(X=1,Y=1,Z=1))))
    WeightMinimum=0
    Weight=&RigVMFunction_MathDoubleSub(A=@n5.OutputParam, B=RigVMVariableNode() {variable=IsCrouchingAndTargetingOrSkydiveFloating}.Value).Result
   )
   n6 = &RigUnit_SphericalPoseReader(
    ActiveRegionScaleFactors=(PositiveWidth=1.7,NegativeWidth=1,PositiveHeight=1.7,NegativeHeight=1)
    ActiveRegionSize=0.1
    Debug=(bDrawDebug=false,bDraw2D=False,bDrawLocalAxes=False,DebugScale=25,DebugSegments=20,DebugThickness=0.25)
    DriverAxis=(X=1,Y=0,Z=0)
    DriverItem=(Type=Bone,Name="head")
    FalloffRegionScaleFactors=(PositiveWidth=1,NegativeWidth=1,PositiveHeight=1,NegativeHeight=1)
    FalloffSize=0.3
    FlipHeightScaling=False
    FlipWidthScaling=False
    OptionalParentItem=(Type=Bone,Name="spine_05")
    RotationOffset=(X=40,Y=0,Z=-130)
    ExecutePin=@RigVMFunction_Sequence.B
   )
   > &RigUnit_ModifyTransforms(
    ItemToModify=((Item=(Type=Bone,Name="dyn_bow"),Transform=(Rotation=(X=0,Y=-0.0872,Z=0,W=0.9962),Translation=(X=-0.5,Y=-0.5,Z=0),Scale3D=(X=1,Y=1,Z=1))))
    WeightMinimum=0
    Weight=&RigVMFunction_MathDoubleMul(A=@n6.OutputParam, B=RigVMVariableNode() {variable=IsCrouching}.Value).Result
   )
   > n7 = &RigUnit_SphericalPoseReader(
    ActiveRegionScaleFactors=(PositiveWidth=1,NegativeWidth=1,PositiveHeight=2.5,NegativeHeight=1)
    ActiveRegionSize=0.05
    Debug=(bDrawDebug=false,bDraw2D=False,bDrawLocalAxes=False,DebugScale=25,DebugSegments=20,DebugThickness=0.25)
    DriverAxis=(X=1,Y=0,Z=0)
    DriverItem=(Type=Bone,Name="head")
    FalloffRegionScaleFactors=(PositiveWidth=0.4,NegativeWidth=1,PositiveHeight=0.5,NegativeHeight=1)
    FalloffSize=0.2
    FlipHeightScaling=False
    FlipWidthScaling=False
    OptionalParentItem=(Type=Bone,Name="spine_05")
    RotationOffset=(X=-76,Y=0,Z=-3)
   )
   n8 = &RigVMFunction_MathDoubleMul(A=@n7.OutputParam, B=RigVMVariableNode() {variable=IsCrouchingNOTMoving}.Value)
   > &RigUnit_ModifyTransforms(
    ItemToModify=((Item=(Type=Bone,Name="dyn_bow"),Transform=(Rotation=(X=0,Y=0,Z=0,W=1),Translation=(X=-3,Y=-2,Z=0),Scale3D=(X=1,Y=1,Z=1))))
    WeightMinimum=0
    Weight=@n8.Result
   )
   > &RigUnit_ModifyTransforms(
    ItemToModify=((Item=(Type=Bone,Name="deform_hair"),Transform=(Rotation=(X=0,Y=0,Z=0,W=1),Translation=(X=0,Y=5,Z=0),Scale3D=(X=1,Y=1,Z=1))))
    WeightMinimum=0
    Weight=@n8.Result
   )
   > n9 = &RigUnit_SphericalPoseReader(
    ActiveRegionScaleFactors=(PositiveWidth=1,NegativeWidth=2,PositiveHeight=2,NegativeHeight=1)
    ActiveRegionSize=0.07
    Debug=(bDrawDebug=false,bDraw2D=False,bDrawLocalAxes=False,DebugScale=25,DebugSegments=20,DebugThickness=0.25)
    DriverAxis=(X=1,Y=0,Z=0)
    DriverItem=(Type=Bone,Name="head")
    FalloffRegionScaleFactors=(PositiveWidth=1,NegativeWidth=5,PositiveHeight=5,NegativeHeight=1)
    FalloffSize=0.15
    FlipHeightScaling=False
    FlipWidthScaling=False
    OptionalParentItem=(Type=Bone,Name="spine_05")
    RotationOffset=(X=-80,Y=0,Z=3)
   )
   n10 = &RigVMFunction_MathDoubleMul(A=@n9.OutputParam, B=RigVMVariableNode() {variable=IsCrouchingMoving}.Value)
   > &RigUnit_ModifyTransforms(
    ItemToModify=((Item=(Type=Bone,Name="dyn_bow"),Transform=(Rotation=(X=0,Y=0,Z=0,W=1),Translation=(X=-3,Y=-2,Z=0),Scale3D=(X=1,Y=1,Z=1))))
    WeightMinimum=0
    Weight=@n10.Result
   )
   > &RigUnit_ModifyTransforms(
    ItemToModify=((Item=(Type=Bone,Name="deform_hair"),Transform=(Rotation=(X=0,Y=0,Z=0,W=1),Translation=(X=0,Y=4,Z=0),Scale3D=(X=1,Y=1,Z=1))))
    WeightMinimum=0
    Weight=@n10.Result
   )
   > &RigUnit_ModifyTransforms(
    ItemToModify=((Item=(Type=Bone,Name="dyn_hair_strand_03_MSTR_back"),Transform=(Rotation=(X=0,Y=0,Z=0,W=1),Translation=(X=-2,Y=0,Z=0),Scale3D=(X=1,Y=1,Z=1))))
    WeightMinimum=0
    Weight=@n10.Result
   )
   > &RigUnit_ModifyTransforms(
    ItemToModify=((Item=(Type=Bone,Name="dyn_hair_strand_04_MSTR_back"),Transform=(Rotation=(X=0,Y=0,Z=0,W=1),Translation=(X=-3,Y=0,Z=0),Scale3D=(X=1,Y=1,Z=1))))
    WeightMinimum=0
    Weight=@n10.Result
   )
   > &RigUnit_ModifyTransforms(
    ItemToModify=((Item=(Type=Bone,Name="dyn_hair_strand_05_MSTR_back"),Transform=(Rotation=(X=0,Y=0,Z=0,W=1),Translation=(X=-3,Y=0,Z=0),Scale3D=(X=1,Y=1,Z=1))))
    WeightMinimum=0
    Weight=@n10.Result
   )
   n11 = &RigUnit_SphericalPoseReader(
    ActiveRegionScaleFactors=(PositiveWidth=1,NegativeWidth=1,PositiveHeight=1,NegativeHeight=1)
    ActiveRegionSize=0.1
    Debug=(bDrawDebug=false,bDraw2D=False,bDrawLocalAxes=False,DebugScale=25,DebugSegments=20,DebugThickness=0.25)
    DriverAxis=(X=1,Y=0,Z=0)
    DriverItem=(Type=Bone,Name="head")
    FalloffRegionScaleFactors=(PositiveWidth=1,NegativeWidth=1,PositiveHeight=1,NegativeHeight=1)
    FalloffSize=0.27
    FlipHeightScaling=False
    FlipWidthScaling=False
    OptionalParentItem=(Type=Bone,Name="spine_05")
    RotationOffset=(X=90,Y=0,Z=-140)
    ExecutePin=@RigVMFunction_Sequence.C
   )
   Multiply = &RigVMFunction_MathDoubleMul(A=@n11.OutputParam, B=RigVMVariableNode() {variable=IsSkydiveFloating}.Value)
   > &RigUnit_ModifyTransforms(
    ItemToModify=((Item=(Type=Bone,Name="dyn_bow"),Transform=(Rotation=(X=0,Y=0,Z=0.0872,W=0.9962),Translation=(X=-1,Y=-1,Z=0),Scale3D=(X=1,Y=1,Z=1))))
    WeightMinimum=0
    Weight=@Multiply.Result
   )
   > n12 = &RigUnit_SphericalPoseReader(
    ActiveRegionScaleFactors=(PositiveWidth=1,NegativeWidth=1,PositiveHeight=1,NegativeHeight=1)
    ActiveRegionSize=0.05
    Debug=(bDrawDebug=false,bDraw2D=False,bDrawLocalAxes=False,DebugScale=25,DebugSegments=20,DebugThickness=0.25)
    DriverAxis=(X=-1,Y=0,Z=0)
    DriverItem=(Type=Bone,Name="clavicle_r")
    FalloffRegionScaleFactors=(PositiveWidth=1,NegativeWidth=1,PositiveHeight=1,NegativeHeight=1)
    FalloffSize=0.22
    FlipHeightScaling=False
    FlipWidthScaling=False
    OptionalParentItem=(Type=Bone,Name="None")
    RotationOffset=(X=50,Y=-130,Z=90)
   )
   > &RigUnit_ModifyTransforms(
    ItemToModify=((Item=(Type=Bone,Name="dyn_bow_01_r_MSTR"),Transform=(Rotation=(X=0,Y=0,Z=-0.1305,W=0.9914),Translation=(X=0,Y=0,Z=0),Scale3D=(X=1,Y=1,Z=1))))
    WeightMinimum=0
    Weight=&RigVMFunction_MathDoubleMul(A=@n12.OutputParam, B=RigVMVariableNode() {variable=IsSkydiveFloating}.Value).Result
   )
   > n13 = &RigUnit_SphericalPoseReader(
    ActiveRegionScaleFactors=(PositiveWidth=1,NegativeWidth=1,PositiveHeight=1,NegativeHeight=1)
    ActiveRegionSize=0.05
    Debug=(bDrawDebug=false,bDraw2D=False,bDrawLocalAxes=False,DebugScale=25,DebugSegments=20,DebugThickness=0.25)
    DriverAxis=(X=1,Y=0,Z=0)
    DriverItem=(Type=Bone,Name="clavicle_l")
    FalloffRegionScaleFactors=(PositiveWidth=1,NegativeWidth=1,PositiveHeight=1,NegativeHeight=1)
    FalloffSize=0.22
    FlipHeightScaling=False
    FlipWidthScaling=False
    OptionalParentItem=(Type=Bone,Name="None")
    RotationOffset=(X=50,Y=-50,Z=-90)
   )
   > &RigUnit_ModifyTransforms(
    ItemToModify=((Item=(Type=Bone,Name="dyn_bow_01_l_MSTR"),Transform=(Rotation=(X=0,Y=0,Z=0.1305,W=0.9914),Translation=(X=0,Y=0,Z=0),Scale3D=(X=1,Y=1,Z=1))))
    WeightMinimum=0
    Weight=&RigVMFunction_MathDoubleMul(A=@n13.OutputParam, B=RigVMVariableNode() {variable=IsSkydiveFloating}.Value).Result
   )
  ]}
]
hierarchy [
 root {loc=<0 0 0>, rot=<0 0 0 1>}
 pelvis {parent=root, loc=<0 -0.5301 98.0571>, rot=<0.0127 0.707 -0.0127 -0.707>}
 spine_01 {parent=pelvis, loc=<6.1208 -0.9328 0>, rot=<0 0 0.034 -0.9994>}
 spine_02 {parent=spine_01, loc=<6.5997 -0.2095 0>, rot=<0 0 0.0238 -0.9997>}
 spine_03 {parent=spine_02, loc=<6.4914 -0.1237 0>, rot=<0 0 0.0149 -0.9999>}
 spine_04 {parent=spine_03, loc=<6.9598 0.0794 0>, rot=<0 0 -0.0132 -0.9999>}
 spine_05 {parent=spine_04, loc=<6.9987 0 0>, rot=<0 0 -0.0015 -1>}
 neck_01 {parent=spine_05, loc=<11.3138 0.1529 0>, rot=<0 0 -0.2003 -0.9797>}
 neck_02 {parent=neck_01, loc=<4.5627 0 0>, rot=<0 0 0.0504 -0.9987>}
 head {parent=neck_02, loc=<5.3914 0 0>, rot=<0 0 0.115 -0.9934>}
 dyn_hair_bangs_01 {parent=head, loc=<7.5237 8.7313 5.8401>, rot=<-0.1426 0.0112 0.986 -0.0855>}
 dyn_hair_bangs_02 {parent=head, loc=<9.1462 10.1924 1.2705>, rot=<-0.1757 -0.0163 0.9808 0.0826>}
 dyn_hair_bangs_03 {parent=head, loc=<9.7168 10.3347 -0.6845>, rot=<-0.2698 0.0461 0.9569 0.097>}
 dyn_hair_bangs_04 {parent=head, loc=<7.5514 10.0944 -3.9869>, rot=<-0.1195 0.2653 0.9553 -0.0519>}
 dyn_hair_strand_01_l {parent=head, loc=<2.5659 6.6511 -6.7103>, rot=<0.1087 0.5282 0.8396 -0.0661>}
 dyn_hair_strand_02_l {parent=head, loc=<1.654 5.4897 -7.5553>, rot=<0.1371 0.4612 0.8705 -0.1036>}
 dyn_hair_strand_03_l {parent=head, loc=<1.0978 2.7687 -8.1839>, rot=<0.1464 0.6342 0.7544 -0.085>}
 dyn_hair_strand_04_l {parent=head, loc=<-1.0924 -1.0468 -7.7426>, rot=<-0.2388 -0.6941 -0.6781 0.0364>}
 dyn_hair_strand_01_r {parent=head, loc=<5.7665 5.181 9.6032>, rot=<0.1765 -0.4761 0.8597 0.0553>}
 dyn_hair_strand_02_r {parent=head, loc=<0.4941 4.7066 7.2793>, rot=<-0.073 -0.5913 0.7644 -0.2464>}
 dyn_hair_strand_03_r {parent=head, loc=<-1.6314 1.3424 7.534>, rot=<-0.1716 -0.6269 0.7555 -0.0818>}
 dyn_hair_strand_04_r {parent=head, loc=<-2.0692 -1.9574 6.6395>, rot=<-0.1087 -0.6416 0.7506 -0.115>}
 dyn_hair_strand_01_MSTR_back {parent=head, loc=<-2.5247 -1.8886 -5.205>, rot=<-0.0571 0.7031 0.7072 -0.0478>}
 dyn_hair_strand_01_back {parent=dyn_hair_strand_01_MSTR_back, loc=<0 0 0>, rot=<0 0 0 1>}
 dyn_hair_strand_02_MSTR_back {parent=head, loc=<-2.5308 -4.2577 -5.8049>, rot=<0.0743 -0.8531 -0.5112 0.073>}
 dyn_hair_strand_02_back {parent=dyn_hair_strand_02_MSTR_back, loc=<0 0 0>, rot=<0 0 0 1>}
 dyn_hair_strand_03_MSTR_back {parent=head, loc=<-1.902 -6.3842 -4.1201>, rot=<0.0499 -0.9371 -0.2994 0.1722>}
 dyn_hair_strand_03_back {parent=dyn_hair_strand_03_MSTR_back, loc=<0 0 0>, rot=<0 0 0 1>}
 dyn_hair_strand_04_MSTR_back {parent=head, loc=<-1.8928 -7.6558 -0.6653>, rot=<0.0138 -0.9809 -0.0952 0.1693>}
 dyn_hair_strand_04_back {parent=dyn_hair_strand_04_MSTR_back, loc=<0 0 0>, rot=<0 0 0 1>}
 dyn_hair_strand_05_MSTR_back {parent=head, loc=<-1.3567 -7.6791 2.639>, rot=<0.0417 -0.9908 0.1154 -0.0565>}
 dyn_hair_strand_05_back {parent=dyn_hair_strand_05_MSTR_back, loc=<0 0 0>, rot=<0 0 0 1>}
 dyn_hair_strand_06_MSTR_back {parent=head, loc=<-2.9517 -5.8472 4.9145>, rot=<-0.0552 -0.9262 0.3661 -0.0712>}
 dyn_hair_strand_06_back {parent=dyn_hair_strand_06_MSTR_back, loc=<0 0 0>, rot=<0 0 0 1>}
 deform_hair_MSTR {parent=head, loc=<5.4465 -7.7721 0>, rot=<0 0 0 1>}
 deform_hair {parent=deform_hair_MSTR, loc=<0 0 0>, rot=<0 0 0 1>}
 faceAttach {parent=head, loc=<-0.1533 -1.0667 0.043>, rot=<0.0029 -0.7071 -0.0029 -0.7071>}
 C_jaw {parent=faceAttach, loc=<0.043 1.3068 -0.8708>, rot=<0 0 0 1>}
 C_lip_lower_mid {parent=C_jaw, loc=<0 8.7198 -2.1929>, rot=<0 0 0 -1>}
 L_lip_lower_outer {parent=C_jaw, loc=<1.1337 8.002 -2.0184>, rot=<0.013 0.0445 0.3057 -0.951>}
 R_lip_lower_outer {parent=C_jaw, loc=<-1.1338 8.002 -2.0184>, rot=<0.013 -0.0445 -0.3057 -0.951>}
 C_lip_upper_mid {parent=faceAttach, loc=<0.0447 10.2107 -2.8079>, rot=<0.2225 0.0013 -0.0005 -0.9749>}
 C_nose_bridge {parent=faceAttach, loc=<0.0254 11.6071 0.7676>, rot=<-0.2218 0 0 -0.9751>}
 L_cheek_inner {parent=faceAttach, loc=<4.2727 8.4876 -0.5202>, rot=<-0.0706 -0.2697 0.3184 -0.906>}
 L_lip_corner {parent=faceAttach, loc=<2.2481 8.9697 -2.942>, rot=<0.003 0.008 -0.0146 -0.9999>}
 L_lip_upper_outer {parent=faceAttach, loc=<1.1937 9.4116 -2.894>, rot=<0.0359 -0.0728 0.2627 -0.9615>}
 R_cheek_inner {parent=faceAttach, loc=<-4.1868 8.4876 -0.5202>, rot=<-0.0706 0.2697 -0.3184 -0.906>}
 R_lip_corner {parent=faceAttach, loc=<-2.1622 8.9697 -2.942>, rot=<0.003 -0.008 0.0146 -0.9999>}
 R_lip_upper_outer {parent=faceAttach, loc=<-1.1078 9.4116 -2.894>, rot=<0.0359 0.0728 -0.2627 -0.9615>}
 teeth_upper {parent=faceAttach, loc=<0.043 9.1232 -1.7072>, rot=<0 0 0 1>}
 clavicle_l {parent=spine_05, loc=<5.4831 -0.1053 -3.2946>, rot=<0.0465 -0.7289 -0.009 -0.6829>}
 clavicle_r {parent=spine_05, loc=<5.4831 -0.1053 3.2946>, rot=<0.7289 0.0465 -0.6829 0.009>}
 dyn_bow_MSTR {parent=spine_05, loc=<13.3417 -4.3541 -0.069>, rot=<0 0 -0.0401 -0.9992>}
 dyn_bow {parent=dyn_bow_MSTR, loc=<0 0 0>, rot=<0 0 0 1>}
 dyn_bow_01_l_MSTR {parent=dyn_bow, loc=<-0.0506 0.2403 -3.6482>, rot=<0.0733 -0.6809 0.012 -0.7286>}
 dyn_bow_01_l {parent=dyn_bow_01_l_MSTR, loc=<0 0 0>, rot=<0 0 0 1>}
 dyn_bow_02_l_MSTR {parent=dyn_bow_01_l, loc=<3.5027 0 0>, rot=<-0.0189 -0.1022 0.2223 -0.9694>}
 dyn_bow_02_l {parent=dyn_bow_02_l_MSTR, loc=<0 0 0>, rot=<0 0 0 1>}
 dyn_bow_01_r_MSTR {parent=dyn_bow, loc=<-0.0509 0.2403 3.7861>, rot=<0.7201 0.0725 -0.6898 -0.016>}
 dyn_bow_01_r {parent=dyn_bow_01_r_MSTR, loc=<0 0 0>, rot=<0 0 0 1>}
 dyn_bow_02_r_MSTR {parent=dyn_bow_01_r, loc=<-3.5027 0 -0.0004>, rot=<-0.0189 -0.1022 0.2223 -0.9694>}
 dyn_bow_02_r {parent=dyn_bow_02_r_MSTR, loc=<0 0 0>, rot=<0 0 0 1>}
]
variables [
 IsCrouching {type=double, default=0}
 IsCrouchingAndTargetingOrSkydiveFloating {type=double, default=0}
 IsCrouchingMoving {type=double, default=0}
 IsCrouchingNOTMoving {type=double, default=0}
 IsSkydiveFloating {type=double, default=0}
]
