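# Post-process style AnimBP for the TrafficHat BobaHat face accessory, adding secondary motion to hair and bows.
# The copied parent pose feeds the Boba variant's Control Rig with crouch, targeting and skydive-float inputs, then gets a fixed cone-bone yaw.
# Chained spring-driven AnimDynamics nodes simulate bangs, side and back strands, and root-relative bows.
# An LOD-driven mask_fix bool blends in cheek and neck bone adjustments; ConeUpAlpha is computed but unused here.
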AnimBlueprint /BRCosmetics/Characters/Player/Female/Medium/Bodies/F_MED_TrafficHat_BobaHat/Meshes/Parts/F_MED_TrafficHat_BobaHat_FaceAcc_AnimBP
target_skeleton=&/F_MED_TrafficHat_BobaHat_FaceAcc_Skeleton
graphs [
 AnimGraph {nodes [
   n1 = Control Rig!(
    Source=Copy Pose From Mesh() {bUseAttachedParent=true}
    Alpha
    IsCrouching:double=Get Is Crouching() -> (Is Crouching:double)
    IsSkydiveFloating:double=Get Is Skydive Floating() -> (Is Skydive Floating:double)
    IsCrouchingAndTargetingOrSkydiveFloating:double=Get Is Crouching and Targeting or Skydive Floating() -> (Is Crouching and Targeting or Skydive Floating:double)
    IsCrouchingMoving:double=Get Is Crouching Moving() -> (Is Crouching Moving:double)
    IsCrouchingNOTMoving:double=Get Is Crouching NOTMoving() -> (Is Crouching NOTMoving:double)
   ) -> (Pose) {
    alphaBoolBlend {blendOption=Linear}
    controlRigAssetReference {blueprintRigClass=/BRCosmetics/Characters/Player/Female/Medium/Bodies/F_MED_TrafficHat_Boba/Meshes/Parts/F_MED_TrafficHat_Boba_FaceAcc_CtrlRig.F_MED_TrafficHat_Boba_FaceAcc_CtrlRig_C}
   }
   n2 = "Transform (Modify) Bone - Bone: dyn_cone_mstr"(ComponentPose=Local To Component(LocalPose=@n1), Alpha=1, Translation="0,0,0", Rotation="0,10,0", Scale="1,1,1") {
    rotationMode=BMM_Additive
    rotationSpace=BCS_BoneSpace
    scaleMode=BMM_Replace
    scaleSpace=BCS_BoneSpace
    alphaBoolBlend {blendOption=Linear}
    boneToModify {boneName=dyn_cone_mstr}
    rotation {yaw=10}
   }
   n3 = Anim Dynamics - Bone: dyn_hair_bangs_01!(ComponentPose=@n2, Alpha=1, GravityOverride:Vector="0,0,0", ExternalForce="0,0,0") -> (Pose) {
    angularDampingOverride=0.9999
    angularSpringConstant=2000
    bAngularSpring=true
    bOverrideAngularDamping=true
    bUseGravityOverride=true
    alphaBoolBlend {blendOption=Linear}
    alphaScaleBiasClamp {bClampResult=true}
    boundBone {boneName=dyn_hair_bangs_01}
    componentAppliedLinearAccClamp=<2000 2000 500>
    componentLinearAccScale=<1 1 1>
    componentLinearVelScale=<1 1 1>
    physicsBodyDefinitions [{
      boundBone {boneName=dyn_hair_bangs_01}
      boxExtents=<10 20 20>
      constraintSetup {angularLimitsMax=<0 20 0>, angularLimitsMin=<0 -20 -40>, angularTarget=<1 0 0>}
      localJointOffset=<5 0 0>
     }]
   }
   n4 = Anim Dynamics - Bone: dyn_hair_bangs_02!(ComponentPose=@n3, Alpha=1, GravityOverride:Vector="0,0,0", ExternalForce="0,0,0") -> (Pose) {
    angularDampingOverride=1
    angularSpringConstant=1000
    bAngularSpring=true
    bOverrideAngularDamping=true
    bUseGravityOverride=true
    alphaBoolBlend {blendOption=Linear}
    alphaScaleBiasClamp {bClampResult=true}
    boundBone {boneName=dyn_hair_bangs_02}
    componentAppliedLinearAccClamp=<2000 2000 1000>
    componentLinearAccScale=<1 1 1>
    componentLinearVelScale=<1 1 1>
    physicsBodyDefinitions [{
      boundBone {boneName=dyn_hair_bangs_02}
      boxExtents=<20 20 20>
      constraintSetup {angularLimitsMax=<0 20 0>, angularLimitsMin=<0 -20 -40>, angularTarget=<1 0 0>}
      localJointOffset=<5 0 0>
     }]
   }
   n5 = Anim Dynamics - Bone: dyn_hair_bangs_03!(ComponentPose=@n4, Alpha=1, GravityOverride:Vector="0,0,0", ExternalForce="0,0,0") -> (Pose) {
    angularDampingOverride=0.9999
    angularSpringConstant=2000
    bAngularSpring=true
    bOverrideAngularDamping=true
    bUseGravityOverride=true
    alphaBoolBlend {blendOption=Linear}
    alphaScaleBiasClamp {bClampResult=true}
    boundBone {boneName=dyn_hair_bangs_03}
    componentAppliedLinearAccClamp=<2000 2000 500>
    componentLinearAccScale=<1 1 1>
    componentLinearVelScale=<1 1 1>
    physicsBodyDefinitions [{
      boundBone {boneName=dyn_hair_bangs_03}
      boxExtents=<10 20 20>
      constraintSetup {angularLimitsMax=<0 20 0>, angularLimitsMin=<0 -20 -40>, angularTarget=<1 0 0>}
      localJointOffset=<4 0 0>
     }]
   }
   n6 = Anim Dynamics - Bone: dyn_hair_bangs_04!(ComponentPose=@n5, Alpha=1, GravityOverride:Vector="0,0,0", ExternalForce="0,0,0") -> (Pose) {
    angularDampingOverride=0.9999
    angularSpringConstant=2000
    bAngularSpring=true
    bOverrideAngularDamping=true
    bUseGravityOverride=true
    alphaBoolBlend {blendOption=Linear}
    alphaScaleBiasClamp {bClampResult=true}
    boundBone {boneName=dyn_hair_bangs_04}
    componentAppliedLinearAccClamp=<2000 2000 1000>
    componentLinearAccScale=<1 1 1>
    componentLinearVelScale=<1 1 1>
    physicsBodyDefinitions [{
      boundBone {boneName=dyn_hair_bangs_04}
      boxExtents=<10 20 20>
      constraintSetup {angularLimitsMax=<0 20 0>, angularLimitsMin=<0 -20 -40>, angularTarget=<1 0 0>}
      localJointOffset=<5 0 0>
     }]
   }
   n7 = Anim Dynamics - Bone: dyn_hair_strand_01_l!(ComponentPose=@n6, Alpha=1, GravityOverride:Vector="0,0,0", ExternalForce="0,0,0") -> (Pose) {
    angularDampingOverride=0.9999
    angularSpringConstant=2000
    bAngularSpring=true
    bOverrideAngularDamping=true
    bUseGravityOverride=true
    alphaBoolBlend {blendOption=Linear}
    alphaScaleBiasClamp {bClampResult=true}
    boundBone {boneName=dyn_hair_strand_01_l}
    componentAppliedLinearAccClamp=<1000 1000 1000>
    componentLinearAccScale=<1 1 1>
    componentLinearVelScale=<1 1 1>
    physicsBodyDefinitions [{
      boundBone {boneName=dyn_hair_strand_01_l}
      boxExtents=<10 20 20>
      constraintSetup {angularLimitsMax=<0 20 0>, angularLimitsMin=<0 -20 -40>, angularTarget=<1 0 0>}
      localJointOffset=<5 0 0>
     }]
   }
   n8 = Anim Dynamics - Bone: dyn_hair_strand_02_l!(ComponentPose=@n7, Alpha=1, GravityOverride:Vector="0,0,0", ExternalForce="0,0,0") -> (Pose) {
    angularDampingOverride=1
    angularSpringConstant=1000
    bAngularSpring=true
    bOverrideAngularDamping=true
    bUseGravityOverride=true
    alphaBoolBlend {blendOption=Linear}
    alphaScaleBiasClamp {bClampResult=true}
    boundBone {boneName=dyn_hair_strand_02_l}
    componentAppliedLinearAccClamp=<500 500 500>
    componentLinearAccScale=<1 1 1>
    componentLinearVelScale=<1 1 1>
    physicsBodyDefinitions [{
      boundBone {boneName=dyn_hair_strand_02_l}
      boxExtents=<20 20 20>
      constraintSetup {angularLimitsMax=<0 20 0>, angularLimitsMin=<0 -20 -40>, angularTarget=<1 0 0>}
      localJointOffset=<5 0 0>
     }]
   }
   n9 = Anim Dynamics - Bone: dyn_hair_strand_03_l!(ComponentPose=@n8, Alpha=1, GravityOverride:Vector="0,0,0", ExternalForce="0,0,0") -> (Pose) {
    angularDampingOverride=0.9999
    angularSpringConstant=2000
    bAngularSpring=true
    bOverrideAngularDamping=true
    bUseGravityOverride=true
    alphaBoolBlend {blendOption=Linear}
    alphaScaleBiasClamp {bClampResult=true}
    boundBone {boneName=dyn_hair_strand_03_l}
    componentAppliedLinearAccClamp=<1000 1000 1000>
    componentLinearAccScale=<1 1 1>
    componentLinearVelScale=<1 1 1>
    physicsBodyDefinitions [{
      boundBone {boneName=dyn_hair_strand_03_l}
      boxExtents=<10 20 20>
      constraintSetup {angularLimitsMax=<0 20 0>, angularLimitsMin=<0 -20 -40>, angularTarget=<1 0 0>}
      localJointOffset=<4 0 0>
     }]
   }
   n10 = Anim Dynamics - Bone: dyn_hair_strand_04_l!(ComponentPose=@n9, Alpha=1, GravityOverride:Vector="0,0,0", ExternalForce="0,0,0") -> (Pose) {
    angularDampingOverride=0.9999
    angularSpringConstant=2000
    bAngularSpring=true
    bOverrideAngularDamping=true
    bUseGravityOverride=true
    alphaBoolBlend {blendOption=Linear}
    alphaScaleBiasClamp {bClampResult=true}
    boundBone {boneName=dyn_hair_strand_04_l}
    componentAppliedLinearAccClamp=<500 500 500>
    componentLinearAccScale=<1 1 1>
    componentLinearVelScale=<1 1 1>
    physicsBodyDefinitions [{
      boundBone {boneName=dyn_hair_strand_04_l}
      boxExtents=<10 20 20>
      constraintSetup {angularLimitsMax=<0 20 0>, angularLimitsMin=<0 -20 -40>, angularTarget=<1 0 0>}
      localJointOffset=<5 0 0>
     }]
   }
   n11 = Anim Dynamics - Bone: dyn_hair_strand_01_r!(ComponentPose=@n10, Alpha=1, GravityOverride:Vector="0,0,0", ExternalForce="0,0,0") -> (Pose) {
    angularDampingOverride=0.9999
    angularSpringConstant=2000
    bAngularSpring=true
    bOverrideAngularDamping=true
    bUseGravityOverride=true
    alphaBoolBlend {blendOption=Linear}
    alphaScaleBiasClamp {bClampResult=true}
    boundBone {boneName=dyn_hair_strand_01_r}
    componentAppliedLinearAccClamp=<1000 1000 1000>
    componentLinearAccScale=<1 1 1>
    componentLinearVelScale=<1 1 1>
    physicsBodyDefinitions [{
      boundBone {boneName=dyn_hair_strand_01_r}
      boxExtents=<10 20 20>
      constraintSetup {angularLimitsMax=<0 20 0>, angularLimitsMin=<0 -20 -40>, angularTarget=<1 0 0>}
      localJointOffset=<5 0 0>
     }]
   }
   n12 = Anim Dynamics - Bone: dyn_hair_strand_02_r!(ComponentPose=@n11, Alpha=1, GravityOverride:Vector="0,0,0", ExternalForce="0,0,0") -> (Pose) {
    angularDampingOverride=1
    angularSpringConstant=1500
    bAngularSpring=true
    bOverrideAngularDamping=true
    bUseGravityOverride=true
    alphaBoolBlend {blendOption=Linear}
    alphaScaleBiasClamp {bClampResult=true}
    boundBone {boneName=dyn_hair_strand_02_r}
    componentAppliedLinearAccClamp=<500 500 500>
    componentLinearAccScale=<1 1 1>
    componentLinearVelScale=<1 1 1>
    physicsBodyDefinitions [{
      boundBone {boneName=dyn_hair_strand_02_r}
      boxExtents=<20 20 20>
      constraintSetup {angularLimitsMax=<0 20 0>, angularLimitsMin=<0 -20 -40>, angularTarget=<1 0 0>}
      localJointOffset=<4 0 0>
     }]
   }
   n13 = Anim Dynamics - Bone: dyn_hair_strand_03_r!(ComponentPose=@n12, Alpha=1, GravityOverride:Vector="0,0,0", ExternalForce="0,0,0") -> (Pose) {
    angularDampingOverride=0.9999
    angularSpringConstant=1500
    bAngularSpring=true
    bOverrideAngularDamping=true
    bUseGravityOverride=true
    alphaBoolBlend {blendOption=Linear}
    alphaScaleBiasClamp {bClampResult=true}
    boundBone {boneName=dyn_hair_strand_03_r}
    componentAppliedLinearAccClamp=<500 500 500>
    componentLinearAccScale=<1 1 1>
    componentLinearVelScale=<1 1 1>
    physicsBodyDefinitions [{
      boundBone {boneName=dyn_hair_strand_03_r}
      boxExtents=<10 20 20>
      constraintSetup {angularLimitsMax=<0 20 0>, angularLimitsMin=<0 -20 -40>, angularTarget=<1 0 0>}
      localJointOffset=<5 0 0>
     }]
   }
   n14 = Anim Dynamics - Bone: dyn_hair_strand_04_r!(ComponentPose=@n13, Alpha=1, GravityOverride:Vector="0,0,0", ExternalForce="0,0,0") -> (Pose) {
    angularDampingOverride=1
    angularSpringConstant=2000
    bAngularSpring=true
    bOverrideAngularDamping=true
    bUseGravityOverride=true
    alphaBoolBlend {blendOption=Linear}
    alphaScaleBiasClamp {bClampResult=true}
    boundBone {boneName=dyn_hair_strand_04_r}
    componentAppliedLinearAccClamp=<1000 1000 1000>
    componentLinearAccScale=<1 1 1>
    componentLinearVelScale=<1 1 1>
    physicsBodyDefinitions [{
      boundBone {boneName=dyn_hair_strand_04_r}
      boxExtents=<15 20 20>
      constraintSetup {angularLimitsMax=<0 20 0>, angularLimitsMin=<0 -20 -40>, angularTarget=<1 0 0>}
      localJointOffset=<4 0 0>
     }]
   }
   n15 = Anim Dynamics - Bone: dyn_hair_strand_01_back!(ComponentPose=@n14, Alpha=1, GravityOverride:Vector="0,0,0", ExternalForce="0,0,0") -> (Pose) {
    angularDampingOverride=0.9999
    angularSpringConstant=2000
    bAngularSpring=true
    bOverrideAngularDamping=true
    bUseGravityOverride=true
    alphaBoolBlend {blendOption=Linear}
    alphaScaleBiasClamp {bClampResult=true}
    boundBone {boneName=dyn_hair_strand_01_back}
    componentAppliedLinearAccClamp=<1000 1000 1000>
    componentLinearAccScale=<1 1 1>
    componentLinearVelScale=<1 1 1>
    physicsBodyDefinitions [{
      boundBone {boneName=dyn_hair_strand_01_back}
      boxExtents=<10 20 20>
      constraintSetup {angularLimitsMax=<0 20 0>, angularLimitsMin=<0 -20 -40>, angularTarget=<1 0 0>}
      localJointOffset=<5 0 0>
     }]
   }
   n16 = Anim Dynamics - Bone: dyn_hair_strand_02_back!(ComponentPose=@n15, Alpha=1, GravityOverride:Vector="0,0,0", ExternalForce="0,0,0") -> (Pose) {
    angularDampingOverride=1
    angularSpringConstant=1000
    bAngularSpring=true
    bOverrideAngularDamping=true
    bUseGravityOverride=true
    alphaBoolBlend {blendOption=Linear}
    alphaScaleBiasClamp {bClampResult=true}
    boundBone {boneName=dyn_hair_strand_02_back}
    componentAppliedLinearAccClamp=<1500 1500 1500>
    componentLinearAccScale=<1 1 1>
    componentLinearVelScale=<1 1 1>
    physicsBodyDefinitions [{
      boundBone {boneName=dyn_hair_strand_02_back}
      boxExtents=<20 20 20>
      constraintSetup {angularLimitsMax=<0 20 0>, angularLimitsMin=<0 -20 -40>, angularTarget=<1 0 0>}
      localJointOffset=<5 0 0>
     }]
   }
   n17 = Anim Dynamics - Bone: dyn_hair_strand_03_back!(ComponentPose=@n16, Alpha=1, GravityOverride:Vector="0,0,0", ExternalForce="0,0,0") -> (Pose) {
    angularDampingOverride=1
    angularSpringConstant=1500
    bAngularSpring=true
    bOverrideAngularDamping=true
    bUseGravityOverride=true
    alphaBoolBlend {blendOption=Linear}
    alphaScaleBiasClamp {bClampResult=true}
    boundBone {boneName=dyn_hair_strand_03_back}
    componentAppliedLinearAccClamp=<1000 1000 1000>
    componentLinearAccScale=<1 1 1>
    componentLinearVelScale=<1 1 1>
    physicsBodyDefinitions [{
      boundBone {boneName=dyn_hair_strand_03_back}
      boxExtents=<20 20 20>
      constraintSetup {angularLimitsMax=<0 20 0>, angularLimitsMin=<0 -20 -40>, angularTarget=<1 0 0>}
      localJointOffset=<4 0 0>
     }]
   }
   n18 = Anim Dynamics - Bone: dyn_hair_strand_04_back!(ComponentPose=@n17, Alpha=1, GravityOverride:Vector="0,0,0", ExternalForce="0,0,0") -> (Pose) {
    angularDampingOverride=0.9999
    angularSpringConstant=2000
    bAngularSpring=true
    bOverrideAngularDamping=true
    bUseGravityOverride=true
    alphaBoolBlend {blendOption=Linear}
    alphaScaleBiasClamp {bClampResult=true}
    boundBone {boneName=dyn_hair_strand_04_back}
    componentAppliedLinearAccClamp=<1000 1000 1000>
    componentLinearAccScale=<1 1 1>
    componentLinearVelScale=<1 1 1>
    physicsBodyDefinitions [{
      boundBone {boneName=dyn_hair_strand_04_back}
      boxExtents=<10 20 20>
      constraintSetup {angularLimitsMax=<0 20 0>, angularLimitsMin=<0 -20 -40>, angularTarget=<1 0 0>}
      localJointOffset=<4 0 0>
     }]
   }
   n19 = Anim Dynamics - Bone: dyn_hair_strand_05_back!(ComponentPose=@n18, Alpha=1, GravityOverride:Vector="0,0,0", ExternalForce="0,0,0") -> (Pose) {
    angularDampingOverride=0.9999
    angularSpringConstant=2000
    bAngularSpring=true
    bOverrideAngularDamping=true
    bUseGravityOverride=true
    alphaBoolBlend {blendOption=Linear}
    alphaScaleBiasClamp {bClampResult=true}
    boundBone {boneName=dyn_hair_strand_05_back}
    componentAppliedLinearAccClamp=<1500 1500 1500>
    componentLinearAccScale=<1 1 1>
    componentLinearVelScale=<1 1 1>
    physicsBodyDefinitions [{
      boundBone {boneName=dyn_hair_strand_05_back}
      boxExtents=<10 20 20>
      constraintSetup {angularLimitsMax=<0 20 0>, angularLimitsMin=<0 -20 -40>, angularTarget=<1 0 0>}
      localJointOffset=<5 0 0>
     }]
   }
   n20 = Anim Dynamics - Bone: dyn_hair_strand_06_back!(ComponentPose=@n19, Alpha=1, GravityOverride:Vector="0,0,0", ExternalForce="0,0,0") -> (Pose) {
    angularDampingOverride=0.9999
    angularSpringConstant=2000
    bAngularSpring=true
    bOverrideAngularDamping=true
    bUseGravityOverride=true
    alphaBoolBlend {blendOption=Linear}
    alphaScaleBiasClamp {bClampResult=true}
    boundBone {boneName=dyn_hair_strand_06_back}
    componentAppliedLinearAccClamp=<1000 1000 1000>
    componentLinearAccScale=<1 1 1>
    componentLinearVelScale=<1 1 1>
    physicsBodyDefinitions [{
      boundBone {boneName=dyn_hair_strand_06_back}
      boxExtents=<10 20 20>
      constraintSetup {angularLimitsMax=<0 20 0>, angularLimitsMin=<0 -20 -40>, angularTarget=<1 0 0>}
      localJointOffset=<5 0 0>
     }]
   }
   n21 = Anim Dynamics - Bone: dyn_bow_01_l!(ComponentPose=@n20, Alpha=1, GravityOverride:Vector="0,0,0", ExternalForce="0,0,0") -> (Pose) {
    angularDampingOverride=1
    angularSpringConstant=15000
    bAngularSpring=true
    bOverrideAngularDamping=true
    bUseGravityOverride=true
    simulationSpace=RootRelative
    alphaBoolBlend {blendOption=Linear}
    boundBone {boneName=dyn_bow_01_l}
    componentAppliedLinearAccClamp=<1000 1000 1000>
    componentLinearAccScale=<0.6 0.6 0.6>
    componentLinearVelScale=<0.7 0.7 0.7>
    physicsBodyDefinitions [{
      boundBone {boneName=dyn_bow_01_l}
      boxExtents=<35 25 25>
      constraintSetup {angularLimitsMax=<0 15 15>, angularLimitsMin=<0 -15 -15>, angularTarget=<1 0 0>}
      localJointOffset=<5 0 0>
     }]
   }
   n22 = Anim Dynamics - Bone: dyn_bow_02_l!(ComponentPose=@n21, Alpha=1, GravityOverride:Vector="0,0,0", ExternalForce="0,0,0") -> (Pose) {
    angularDampingOverride=0.9998
    angularSpringConstant=10000
    bAngularSpring=true
    bOverrideAngularDamping=true
    bUseGravityOverride=true
    simulationSpace=RootRelative
    alphaBoolBlend {blendOption=Linear}
    boundBone {boneName=dyn_bow_02_l}
    componentAppliedLinearAccClamp=<1000 1000 1000>
    componentLinearAccScale=<0.5 0.5 0.5>
    componentLinearVelScale=<0.6 0.6 0.6>
    physicsBodyDefinitions [{
      boundBone {boneName=dyn_bow_02_l}
      boxExtents=<35 20 20>
      constraintSetup {angularLimitsMax=<0 15 15>, angularLimitsMin=<0 -15 -15>, angularTarget=<1 0 0>}
      localJointOffset=<3 0 0>
     }]
   }
   n23 = Anim Dynamics - Bone: dyn_bow_01_r!(ComponentPose=@n22, Alpha=1, GravityOverride:Vector="0,0,0", ExternalForce="0,0,0") -> (Pose) {
    angularDampingOverride=1
    angularSpringConstant=20000
    bAngularSpring=true
    bOverrideAngularDamping=true
    bUseGravityOverride=true
    bUseSphericalLimits=true
    simulationSpace=RootRelative
    alphaBoolBlend {blendOption=Linear}
    boundBone {boneName=dyn_bow_01_r}
    componentAppliedLinearAccClamp=<1500 1500 1500>
    componentLinearAccScale=<1 1 1>
    componentLinearVelScale=<1 1 1>
    physicsBodyDefinitions [{
      sphereCollisionRadius=1
      boundBone {boneName=dyn_bow_01_r}
      boxExtents=<40 30 30>
      constraintSetup {angularLimitsMax=<0 15 15>, angularLimitsMin=<0 -15 -15>, angularTarget=<1 0 0>}
      localJointOffset=<-6 0 0>
     }]
   }
   n24 = Anim Dynamics - Bone: dyn_bow_02_r!(ComponentPose=@n23, Alpha=1, GravityOverride:Vector="0,0,0", ExternalForce="0,0,0") -> (Pose) {
    angularDampingOverride=0.9998
    angularSpringConstant=10000
    bAngularSpring=true
    bOverrideAngularDamping=true
    bUseGravityOverride=true
    simulationSpace=RootRelative
    alphaBoolBlend {blendOption=Linear}
    boundBone {boneName=dyn_bow_02_r}
    componentAppliedLinearAccClamp=<1000 1000 1000>
    componentLinearAccScale=<1 1 1>
    componentLinearVelScale=<1 1 1>
    physicsBodyDefinitions [{
      boundBone {boneName=dyn_bow_02_r}
      boxExtents=<35 30 30>
      constraintSetup {angularLimitsMax=<0 15 15>, angularLimitsMin=<0 -15 -15>, angularTarget=<1 0 0>}
      localJointOffset=<-3 0 0>
     }]
   }
   n25 = "Transform (Modify) Bone - Bone: L_cheek_inner"!(ComponentPose=@n24, bAlphaBoolEnabled:bool=Get mask_fix() -> (mask_fix:bool), Translation="0,0.1,0", Rotation="0,0,0", Scale="1.2,1.2,1.2") -> (Pose) {
    alphaInputType=Bool
    scaleMode=BMM_Additive
    scaleSpace=BCS_BoneSpace
    translationMode=BMM_Additive
    translationSpace=BCS_BoneSpace
    alphaBoolBlend {blendInTime=0.3, blendOption=Linear, blendOutTime=0.6}
    boneToModify {boneName=L_cheek_inner}
    scale=<1.2 1.2 1.2>
    translation {y=0.1}
   }
   n26 = "Transform (Modify) Bone - Bone: R_cheek_inner"!(ComponentPose=@n25, bAlphaBoolEnabled:bool=Get mask_fix(), Translation="0,0.1,0", Rotation="0,0,0", Scale="1.2,1.2,1.2") -> (Pose) {
    alphaInputType=Bool
    scaleMode=BMM_Additive
    scaleSpace=BCS_BoneSpace
    translationMode=BMM_Additive
    translationSpace=BCS_BoneSpace
    alphaBoolBlend {blendInTime=0.3, blendOption=Linear, blendOutTime=0.6}
    boneToModify {boneName=R_cheek_inner}
    scale=<1.2 1.2 1.2>
    translation {y=0.1}
   }
   n27 = "Transform (Modify) Bone - Bone: neck_02_mask_fix"!(ComponentPose=@n26, bAlphaBoolEnabled:bool=Get mask_fix(), Translation="0,0,0", Rotation="0,0,0", Scale="1.1,1.1,1.1") -> (Pose) {
    alphaInputType=Bool
    scaleMode=BMM_Additive
    scaleSpace=BCS_BoneSpace
    alphaBoolBlend {blendInTime=0.3, blendOption=Linear, blendOutTime=0.6}
    boneToModify {boneName=neck_02_mask_fix}
    scale=<1.1 1.1 1.1>
   }
   Output Pose(Result=Component To Local(ComponentPose=@n27)) {layerGroup=None}
  ]}
 EventGraph {nodes [
   Event BlueprintUpdateAnimation()
   > Set Delta Time X(Delta Time X=0)
   > n1 = Gravity Override State Function DEPRECATED(CharacterPartAnimInstance=Self-Reference(), GameTimeInSeconds=GetGameTimeInSeconds(), GravityOverrideParams=Get Gravity Override Params())
   > Set Gravity Override(Gravity Override=@n1)
   n2 = Get bIsCrouching()
   > Set Is Crouching(Is Crouching:double=FInterpTo(Current=Get Is Crouching(), Target="To Float (Boolean)"(InBool=@n2), DeltaTime=Get Delta Time X(), InterpSpeed=10)) -> (Output_Get:double)
   n3 = AND Boolean(A=Get bIsCrouching(), B=Get bIsMoving2D())
   > Set Is Crouching Moving(Is Crouching Moving:double=FInterpTo(Current=Get Is Crouching Moving(), Target="To Float (Boolean)"(InBool=@n3), DeltaTime=Get Delta Time X(), InterpSpeed=5)) -> (Output_Get:double)
   n4 = AND Boolean(A=Get bIsCrouching(), B=NOT Boolean(A=Get bIsMoving2D()))
   > Set Is Crouching NOTMoving(Is Crouching NOTMoving:double=FInterpTo(Current=Get Is Crouching NOTMoving(), Target="To Float (Boolean)"(InBool=@n4), DeltaTime=Get Delta Time X(), InterpSpeed=5)) -> (Output_Get:double)
   n5 = OR Boolean(A=AND Boolean(A=Get bIsCrouching(), B=Get bIsTargeting()), B=Get bIsSkydiveFloating())
   > Set Is Crouching and Targeting or Skydive Floating(Is Crouching and Targeting or Skydive Floating:double=FInterpTo(
     Current=Get Is Crouching and Targeting or Skydive Floating()
     Target="To Float (Boolean)"(InBool=@n5)
     DeltaTime=Get Delta Time X()
     InterpSpeed=10
    )) -> (Output_Get:double)
   n6 = Get bIsSkydiveFloating()
   > Set Is Skydive Floating(Is Skydive Floating:double=FInterpTo(Current=Get Is Skydive Floating(), Target="To Float (Boolean)"(InBool=@n6), DeltaTime=Get Delta Time X(), InterpSpeed=10)) -> (Output_Get:double)
   > Set ConeUpAlpha(ConeUpAlpha:float=GetAlphaFromReferenceTransformDelta(SkeletalMeshComponent=GetOwnerSkeletalMeshComponent(), BoneName=head, BaseName=spine_05, InputRange=(X=-15,Y=-55), TwistAxis="1,0,0")) -> (Output_Get:float)
   > Set mask_fix(mask_fix:bool="integer > integer"(A=Get CurrentLOD())) -> (Output_Get:bool)
  ]}
]
variables [
 ConeUpAlpha {type=number}
 Delta Time X {type=number}
 Gravity Override {type=Vector}
 Gravity Override Params {type=GravityOverrideParamsStruct}
 Is Crouching {type=number}
 Is Crouching and Targeting or Skydive Floating {type=number}
 Is Crouching Moving {type=number}
 Is Crouching NOTMoving {type=number}
 Is Skydive Floating {type=number}
 mask_fix {type=boolean}
]
